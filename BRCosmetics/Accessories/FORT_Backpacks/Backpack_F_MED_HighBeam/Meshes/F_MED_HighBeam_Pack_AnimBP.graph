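# Post-process style AnimBlueprint for the HighBeam backpack, adjusting pack pose and hiding elements as matching cosmetics come into use.
# AnimGraph applies spine-yaw-driven offsets to the pack bone and a fixed root offset.
# Anim Dynamics chain for crossbow and two swords is not connected to the output pose.
# EventGraph toggles "Hide Element" material scalars from pickaxe, movement, front-end idle and glider state, delaying the glider hide.
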AnimBlueprint /BRCosmetics/Accessories/FORT_Backpacks/Backpack_F_MED_HighBeam/Meshes/F_MED_HighBeam_Pack_AnimBP
target_skeleton=&/F_MED_HighBeam_Pack_Skeleton
graphs [
 AnimGraph {nodes [
   n1 = Get Dyn Strength() -> (Dyn Strength:double)
   n2 = Anim Dynamics - Bone: dyn_crossbow(Alpha=@n1, ExternalForce="0,0,0") {
    angularDampingOverride=1
    angularSpringConstant=1000
    bAngularSpring=true
    bOverrideAngularDamping=true
    bUseGravityOverride=true
    gravityScale=0.35
    simulationSpace=RootRelative
    alphaBoolBlend {blendOption=Linear}
    boundBone {boneName=dyn_crossbow}
    componentLinearAccScale=<1 1 1>
    componentLinearVelScale=<1 1 1>
    gravityOverride {z=-980}
    physicsBodyDefinitions [{
      boundBone {boneName=dyn_crossbow}
      boxExtents=<30 30 30>
      constraintSetup {angularLimitsMax=<8 8 5>, angularLimitsMin=<-5 -4 -10>, angularTarget=<1 0 0>}
      localJointOffset=<1 0 0>
     }]
   }
   n3 = Anim Dynamics - Bone: dyn_sword_steel(ComponentPose=@n2, Alpha=@n1, ExternalForce="0,0,0") {
    angularDampingOverride=1
    angularSpringConstant=2000
    bAngularSpring=true
    bOverrideAngularDamping=true
    bUseGravityOverride=true
    gravityScale=0.35
    simulationSpace=RootRelative
    alphaBoolBlend {blendOption=Linear}
    boundBone {boneName=dyn_sword_steel}
    componentLinearAccScale=<1 1 1>
    componentLinearVelScale=<1 1 1>
    gravityOverride {z=-980}
    physicsBodyDefinitions [{
      boundBone {boneName=dyn_sword_steel}
      boxExtents=<50 50 50>
      constraintSetup {angularLimitsMax=<5 3 5>, angularLimitsMin=<-5 -3 -3>, angularTarget=<1 0 0>}
      localJointOffset=<1 0 0>
     }]
   }
   Anim Dynamics - Bone: dyn_sword_silver(ComponentPose=@n3, Alpha=@n1, ExternalForce="0,0,0") {
    angularDampingOverride=1
    angularSpringConstant=2000
    bAngularSpring=true
    bOverrideAngularDamping=true
    bUseGravityOverride=true
    gravityScale=0.35
    simulationSpace=RootRelative
    alphaBoolBlend {blendOption=Linear}
    boundBone {boneName=dyn_sword_silver}
    componentLinearAccScale=<1 1 1>
    componentLinearVelScale=<1 1 1>
    gravityOverride {z=-980}
    physicsBodyDefinitions [{
      boundBone {boneName=dyn_sword_silver}
      boxExtents=<50 50 50>
      constraintSetup {angularLimitsMax=<5 3 4>, angularLimitsMin=<-5 -3 -2>, angularTarget=<1 0 0>}
      localJointOffset=<1 0 0>
     }]
   }
   n4 = "Transform (Modify) Bone - Bone: pack"(ComponentPose=Local To Component(LocalPose=Input Pose()), Translation=Get backpack_translation(), Rotation=Get backpack_rotate()) {
    rotationMode=BMM_Additive
    scaleMode=BMM_Additive
    scaleSpace=BCS_BoneSpace
    translationMode=BMM_Additive
    alphaBoolBlend {blendOption=Linear}
    boneToModify {boneName=pack}
   }
   n5 = "Transform (Modify) Bone - Bone: root"(ComponentPose=@n4, Translation="-2.5,1.5,-4.5", Rotation="0,-10,0") {
    rotationMode=BMM_Additive
    rotationSpace=BCS_BoneSpace
    scaleMode=BMM_Additive
    scaleSpace=BCS_BoneSpace
    translationMode=BMM_Additive
    translationSpace=BCS_BoneSpace
    alphaBoolBlend {blendOption=Linear}
    boneToModify {boneName=root}
    rotation {yaw=-10}
    translation=<-2.5 1.5 -4.5>
   }
   Output Pose(Result=Component To Local(ComponentPose=@n5)) {layerGroup=None}
  ]}
 EventGraph {nodes [
   Event BlueprintUpdateAnimation()
   > n1 = Sequence!(execute) -> (then_0, then_1, then_2:Exec)
   n2 = GetOwnerSkeletalMeshComponent()
   n3 = BreakRotator(InRot=BreakTransform(InTransform=GetDeltaTransformFromRefPose(self=@n2, BoneName=spine_05, BaseName=spine_01)).Rotation)
   > Set backpack_rotate(backpack_rotate=MakeRotator(Roll=MapRangeClamped(Value=@n3.Yaw, InRangeA=0, InRangeB=-30, OutRangeA=5, OutRangeB=25)))
   > Set backpack_translation(backpack_translation=MakeVector(Y=MapRangeClamped(Value=@n3.Yaw, InRangeA=0, InRangeB=-30, OutRangeB=-6.5)))
   Branch(execute=@n1.then_1, Condition=Get bIsInFrontEnd())
   n4 = Make Array("[0]"=(TagName="Cosmetics.MetaTags.HighBeam"))
   n5 = GetCurveValue(self=GetOwnerFrontEndAnimInstance(), CurveName=FrontEnd_Idle_Bypass)
   > n6 = Branch(Condition=AND Boolean(A=CharacterItemDefinitionHasAnyMetaTag(Tags=@n4), B="float >= float"(A=@n5, B=1)))
   > Set FE Idle Active(FE Idle Active:bool=true) -> (Output_Get:bool)
   Set FE Idle Active(execute=@n6.else, FE Idle Active=false)
   n7 = Get Dynamic Material Instance(execute=[@n16.then, @n19.then, @n1.then_2], SkeletalMeshComponent=GetPartSkeletalMeshComponent())
   n8 = PickaxeItemDefinitionHasAnyMetaTag(Tags=Make Array("[0]"=(TagName="Cosmetics.MetaTags.HighBeam")), bRequiresEquipped=true)
   > n9 = Branch(Condition=@n8)
   > n10 = Branch(Condition=OR Boolean!(A=Get bIsSkyDiving(), B=Get bIsTacticalSprinting(), C:bool=Get bIsPlayingEmote()) -> (ReturnValue))
   > n11 = Branch(Condition=Get bIsCrouchSprinting())
   SetScalarParameterValue(execute=[@n14.then_0, @n15.then, @n21.then, @n11.then, @n10.else], self=@n7, ParameterName=Hide Element 01, Value=1)
   n12 = Branch(execute=[@n21.else, @n11.else, @n9.else], Condition=Get bIsParachuteOpened())
   n13 = Branch(
    execute=[@n17.then, @n12.then]
    Condition=GliderItemDefinitionHasAnyMetaTag!(Tags:Array of Gameplay Tag Structures=Make Array("[0]"=(TagName="Cosmetics.MetaTags.HighBeamGlider")), bRequiresOpened:bool=true) -> (ReturnValue:bool)
   )
   > Do Once(Reset=@n18.then_0)
   > Delay(Duration=0.7)
   > n14 = Sequence!(execute) -> (then_0, then_1)
   SetScalarParameterValue(execute=@n14.then_1, self=@n7, ParameterName=Hide Element 02, Value=1)
   n15 = Branch(execute=[@n20.then, @n18.then_1, @n12.else, @n13.else], Condition=Get FE Idle Active() -> (FE Idle Active:bool))
   SetScalarParameterValue(execute=[@n15.else, @n22.OutputPin], self=@n7, ParameterName=Hide Element 01, Value=0)
   > SetScalarParameterValue(self=@n7, ParameterName=Hide Element 02, Value=0)
   n16 = Event OnCharacterCustomizationLoaded()
   n17 = Event OnParachuteOpened()
   Event OnParachuteClosed()
   > n18 = Sequence()
   n19 = Event OnWeaponChanged()
   n20 = Event OnBeginSkydiving()
   Event OnEndSkydiving()
   > n21 = Branch(Condition=@n8)
   n22 = Reroute Node()
  ]}
]
variables [
 backpack_rotate {type=Rotator}
 backpack_translation {type=Vector}
 FE Idle Active {type=boolean}
 Is_Backpack_Equipped {type=boolean}
 PickaxeEquipped {type=number}
]
